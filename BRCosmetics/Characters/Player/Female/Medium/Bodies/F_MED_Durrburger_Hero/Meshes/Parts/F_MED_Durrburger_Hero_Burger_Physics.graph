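# Physics asset for the burger head part of the female Durrburger Hero character cosmetic, giving collision bodies and joints to its dynamic bones.
# A three-box tongue chain hangs from the head, each segment constrained to the one before it with matching swing and twist limits.
# Capsules for the top and three cheese pieces constrain directly to the head, with the top held tightly and each cheese piece given its own asymmetric swing limits.
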PhysicsAsset /BRCosmetics/Characters/Player/Female/Medium/Bodies/F_MED_Durrburger_Hero/Meshes/Parts/F_MED_Durrburger_Hero_Burger_Physics
bodies {
 dyn_cheese_bk [{shapeType=Capsule, length=2.75, radius=1, center=<2.0019 0 0>, rotation=<rot 90 0 0>}]
 dyn_cheese_l [{shapeType=Capsule, length=2.75, radius=1, center=<2.1082 0 0>, rotation=<rot 90 0 0>}]
 dyn_cheese_r [{shapeType=Capsule, length=2.75, radius=1, center=<1.6399 0 0>, rotation=<rot 90 0 0>}]
 dyn_tongue_01 [{shapeType=Box, extentX=5, extentY=13.5, extentZ=2.5, center=<3.6356 0.0001 0>}]
 dyn_tongue_02 [{shapeType=Box, extentX=6, extentY=12.5, extentZ=2.5, center=<3.2179 0 0.0001>}]
 dyn_tongue_03 [{shapeType=Box, extentX=6, extentY=12, extentZ=2.5, center=<3.1157 0 0>}]
 dyn_top [{shapeType=Capsule, length=7.5, radius=3.75, center=<4.6704 0.0001 0>, rotation=<rot 90 0 0>}]
 head [{shapeType=Capsule, length=1, radius=1}]
}
constraints [
 {bone1Name=dyn_tongue_01, bone2Name=head, swing1LimitDegrees=30, swing2LimitDegrees=30, twistLimitDegrees=10}
 {bone1Name=dyn_tongue_02, bone2Name=dyn_tongue_01, swing1LimitDegrees=30, swing2LimitDegrees=30, twistLimitDegrees=10}
 {bone1Name=dyn_tongue_03, bone2Name=dyn_tongue_02, swing1LimitDegrees=30, swing2LimitDegrees=30, twistLimitDegrees=10}
 {bone1Name=dyn_top, bone2Name=head, swing1LimitDegrees=2, swing2LimitDegrees=10, twistLimitDegrees=2}
 {bone1Name=dyn_cheese_l, bone2Name=head, swing1LimitDegrees=10, swing2LimitDegrees=30, twistLimitDegrees=5}
 {bone1Name=dyn_cheese_r, bone2Name=head, swing1LimitDegrees=30, swing2LimitDegrees=10, twistLimitDegrees=5}
 {bone1Name=dyn_cheese_bk, bone2Name=head, swing1LimitDegrees=20, swing2LimitDegrees=30, twistLimitDegrees=5}
]
