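# Physics asset for the dangling necklace on the Durrburger Hero female medium body, defining its collision bodies and joint limits.
# Capsules on spine_04 and spine_05 form the torso volume, with a small capsule at the necklace base.
# Thin boxes form a chain from the base through four segments to the tag.
# Spine and base joints swing and twist freely, while chain limits start tight at the first segment and loosen toward the tag.
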PhysicsAsset /BRCosmetics/Characters/Player/Female/Medium/Bodies/F_MED_Durrburger_Hero/Meshes/Parts/F_MED_Durrburger_Hero_Necklace_Physics
bodies {
 dyn_necklace_a [{shapeType=Box, extentX=2.314, extentY=5.6527, extentZ=11.5873, center=<2.1366 -0.5216 -0.078>, rotation=<rot 0.0763 -90 -1.3319>}]
 dyn_necklace_a_base [{shapeType=Capsule, length=0.505, radius=0.505, rotation=<rot 0 -90 0>}]
 dyn_necklace_b [{shapeType=Box, extentX=1.7364, extentY=5.0464, extentZ=12.6142, center=<2.4252 0.552 -0.0089>, rotation=<rot 0.198 -90 -0.04>}]
 dyn_necklace_c [{shapeType=Box, extentX=1.2661, extentY=5.4427, extentZ=10.6589, center=<2.8433 0.0281 -0.0669>, rotation=<rot 1.4363 -90 1.7789>}]
 dyn_necklace_d [{shapeType=Box, extentX=1.4519, extentY=4.5196, extentZ=7.7503, center=<2.0238 -0.1883 -0.2621>, rotation=<rot 2.1706 -90 3.2763>}]
 dyn_necklace_tag [{shapeType=Box, extentX=0.5, extentY=2.6357, extentZ=5.4779, center=<2.1304 0.0044 0.2619>, rotation=<rot 0 89.92 -98.8525>}]
 spine_04 [
  {shapeType=Capsule, length=3, radius=12.5, center=<-0.0787 0.1664 0>}
  {shapeType=Capsule, length=0, radius=9, center=<-1.0579 4.9353 -4.5482>, rotation=<rot 0 0 93.9923>}
  {shapeType=Capsule, length=0, radius=9.75, center=<-1.3785 4.1765 4.1321>, rotation=<rot 0 0 93.9923>}
 ]
 spine_05 [{shapeType=Capsule, length=15.25, radius=11.25, center=<-1.9654 -2.1467 0.0002>}]
}
constraints [
 {
  bone1Name=spine_05
  bone2Name=spine_04
  swing1LimitDegrees=45
  swing1Motion=Free
  swing2LimitDegrees=45
  swing2Motion=Free
  twistLimitDegrees=45
  twistMotion=Free
 }
 {
  bone1Name=dyn_necklace_a_base
  bone2Name=spine_05
  swing1LimitDegrees=45
  swing1Motion=Free
  swing2LimitDegrees=45
  swing2Motion=Free
  twistLimitDegrees=45
  twistMotion=Free
 }
 {bone1Name=dyn_necklace_a, bone2Name=dyn_necklace_a_base, swing1LimitDegrees=5, swing2LimitDegrees=5, twistLimitDegrees=5}
 {bone1Name=dyn_necklace_b, bone2Name=dyn_necklace_a, swing1LimitDegrees=30, swing2LimitDegrees=30, twistLimitDegrees=10}
 {bone1Name=dyn_necklace_c, bone2Name=dyn_necklace_b, swing1LimitDegrees=45, swing2LimitDegrees=30, twistLimitDegrees=15}
 {bone1Name=dyn_necklace_d, bone2Name=dyn_necklace_c, swing1LimitDegrees=45, swing2LimitDegrees=30, twistLimitDegrees=15}
 {bone1Name=dyn_necklace_tag, bone2Name=dyn_necklace_d, swing1LimitDegrees=45, swing2LimitDegrees=45, twistLimitDegrees=15}
]
